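# Animation Blueprint for the grenades part of the male medium Ultra Rare Commando body cosmetic, adding secondary physics motion to two grenade bones.
# On initialization it caches the owner's skeletal mesh component, whose pose the graph copies.
# Two chained Anim Dynamics nodes then simulate dyn_grenade02 and dyn_grenade01 within tight limits, using a pelvis-driven sphere limit.
# Only grenade02 enables an angular spring, while grenade01 runs at reduced gravity.
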AnimBlueprint /BRCosmetics/Characters/Player/Male/Medium/Bodies/M_Med_UltraRareCommando_01/Meshes/Parts/M_MED_Commando_UltraRare_01_Grenades_AnimBP
target_skeleton=&/M_MED_Commando_UltraRare_01_Grenades_Skeleton
graphs [
 AnimGraph {nodes [
   n1 = Get MeshToCopy() -> (MeshToCopy:Skeletal Mesh Component Object Reference)
   n2 = Anim Dynamics - Bone: dyn_grenade02(ComponentPose=Local To Component(LocalPose=Copy Pose From Mesh(SourceMeshComponent=@n1)), Alpha=1, ExternalForce="0,0,0") {
    angularSpringConstant=1000
    bAngularSpring=true
    bGravityOverrideInSimSpace=true
    lODThreshold=1
    numSolverIterationsPostUpdate=4
    numSolverIterationsPreUpdate=16
    alphaBoolBlend {blendOption=Linear}
    boundBone {boneName=dyn_grenade02}
    physicsBodyDefinitions [{
      sphereCollisionRadius=0
      boundBone {boneName=dyn_grenade02}
      boxExtents=<10 10 20>
      constraintSetup {
       angularTargetAxis=AxisY
       twistAxis=AxisZ
       angularLimitsMax=<20 0 0>
       angularLimitsMin=<0 -20 0>
       angularTarget=<0 0 0>
       linearAxesMax=<0 0 0.5>
       linearAxesMin=<0 0 -0.5>
      }
      localJointOffset=<0 0 -2>
     }]
    sphericalLimits [{limitRadius=16, drivingBone {boneName=pelvis}, sphereLocalOffset=<-1 1 0>}]
   }
   n3 = Anim Dynamics - Bone: dyn_grenade01(ComponentPose=@n2, Alpha=1, ExternalForce="0,0,0") {
    angularSpringConstant=5000
    bGravityOverrideInSimSpace=true
    gravityScale=0.2
    lODThreshold=1
    numSolverIterationsPostUpdate=4
    numSolverIterationsPreUpdate=16
    alphaBoolBlend {blendOption=Linear}
    boundBone {boneName=dyn_grenade01}
    physicsBodyDefinitions [{
      sphereCollisionRadius=0
      boundBone {boneName=dyn_grenade01}
      boxExtents=<10 10 20>
      constraintSetup {
       angularTargetAxis=AxisY
       twistAxis=AxisZ
       angularLimitsMax=<20 0 0>
       angularLimitsMin=<0 -20 0>
       angularTarget=<0 0 0>
       linearAxesMax=<0 0 0.5>
       linearAxesMin=<0 0 -0.5>
      }
      localJointOffset=<0 0 -2>
     }]
    sphericalLimits [{limitRadius=16, drivingBone {boneName=pelvis}, sphereLocalOffset=<-1 1 0>}]
   }
   Output Pose(Result=Component To Local(ComponentPose=@n3)) {layerGroup=None}
  ]}
 EventGraph {nodes [
   Event BlueprintInitializeAnimation()
   > Set MeshToCopy(MeshToCopy:Skeletal Mesh Component Object Reference=GetOwnerSkeletalMeshComponent()) -> (Output_Get:Skeletal Mesh Component Object Reference)
   Event BlueprintUpdateAnimation()
  ]}
]
variables [MeshToCopy {type=SkeletalMeshComponent}]
